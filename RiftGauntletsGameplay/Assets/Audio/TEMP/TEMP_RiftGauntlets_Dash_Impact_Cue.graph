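# Temporary Sound Cue for the Rift Gauntlets dash impact, routed to the explosives sound class with weapon explosion concurrency.
# It mixes close and distant branches, each with its own attenuation, with the close branch weighted louder.
# Each branch layers a random impact with a slightly delayed random zap, quieter in the distant mix.
# Quality level nodes switch each layer between its full variation pool and a three-sample subset.
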SoundCue /RiftGauntletsGameplay/Assets/Audio/TEMP/TEMP_RiftGauntlets_Dash_Impact_Cue
root=Mixer_4
nodes [
 {
  id=Attenuation_8
  children [Modulator_1]
  properties {attenuationSettings=/RiftGauntletsGameplay/Assets/Audio/TEMP/Settings/TEMP_RiftGauntlets_Dash_Impact_Close_Att}
 }
 {
  id=Attenuation_9
  children [Mixer_1]
  properties {attenuationSettings=/RiftGauntletsGameplay/Assets/Audio/TEMP/Settings/TEMP_RiftGauntlets_Dash_Impact_Distant_Att}
 }
 {id=Delay_0, children [Modulator_0], properties {delayMax=0.1, delayMin=0.1}}
 {id=Delay_1, children [Modulator_2], properties {delayMax=0.1, delayMin=0.1}}
 {id=Mixer_0, children [Modulator_8,Delay_0], properties {inputVolume [1,1]}}
 {id=Mixer_1, children [Modulator_10,Delay_1], properties {inputVolume [1,0.3]}}
 {id=Mixer_4, children [Attenuation_8, Attenuation_9], properties {inputVolume [1.5,0.5]}}
 {id=Modulator_0, children [QualityLevel_0], properties {pitchMax=1, pitchMin=1}}
 {id=Modulator_1, children [Mixer_0]}
 {id=Modulator_10, children [QualityLevel_7]}
 {id=Modulator_2, children [QualityLevel_1], properties {pitchMax=1, pitchMin=1}}
 {id=Modulator_8, children [QualityLevel_6]}
 {id=QualityLevel_0, children [Random_0,Random_0,Random_1,Random_1]}
 {id=QualityLevel_1, children [Random_2,Random_2,Random_3,Random_3]}
 {id=QualityLevel_6, children [Random_6,Random_6,Random_14,Random_14]}
 {id=QualityLevel_7, children [Random_15,Random_15,Random_16,Random_16]}
 {
  id=Random_0
  children [WavePlayer_20, WavePlayer_21, WavePlayer_22, WavePlayer_23, WavePlayer_24, WavePlayer_25]
  properties {weights [1,1,1,1,1,1]}
 }
 {id=Random_1, children [WavePlayer_23, WavePlayer_24, WavePlayer_25], properties {bShouldExcludeFromBranchCulling=true, weights [1,1,1]}}
 {id=Random_14, children [WavePlayer_7,WavePlayer_8,WavePlayer_9], properties {bShouldExcludeFromBranchCulling=true, weights [1,1,1]}}
 {
  id=Random_15
  children [
   WavePlayer_10
   WavePlayer_11
   WavePlayer_12
   WavePlayer_13
   WavePlayer_14
   WavePlayer_15
   WavePlayer_16
   WavePlayer_17
   WavePlayer_18
   WavePlayer_19
  ]
  properties {weights [1,1,1,1,1,1,1,1,1,1]}
 }
 {id=Random_16, children [WavePlayer_17, WavePlayer_18, WavePlayer_19], properties {bShouldExcludeFromBranchCulling=true, weights [1,1,1]}}
 {
  id=Random_2
  children [WavePlayer_26, WavePlayer_27, WavePlayer_28, WavePlayer_29, WavePlayer_30, WavePlayer_31]
  properties {weights [1,1,1,1,1,1]}
 }
 {id=Random_3, children [WavePlayer_29, WavePlayer_30, WavePlayer_31], properties {bShouldExcludeFromBranchCulling=true, weights [1,1,1]}}
 {
  id=Random_6
  children [WavePlayer_0,WavePlayer_1,WavePlayer_2,WavePlayer_3,WavePlayer_4,WavePlayer_5,WavePlayer_6,WavePlayer_7,WavePlayer_8,WavePlayer_9]
  properties {weights [1,1,1,1,1,1,1,1,1,1]}
 }
 {id=WavePlayer_0, properties {soundWaveAssetPtr=&/TEMP_RiftGauntlets_Dash_Impact_Close_01}}
 {id=WavePlayer_1, properties {soundWaveAssetPtr=&/TEMP_RiftGauntlets_Dash_Impact_Close_02}}
 {id=WavePlayer_10, properties {soundWaveAssetPtr=&/TEMP_RiftGauntlets_Dash_Impact_Distant_01}}
 {id=WavePlayer_11, properties {soundWaveAssetPtr=&/TEMP_RiftGauntlets_Dash_Impact_Distant_02}}
 {id=WavePlayer_12, properties {soundWaveAssetPtr=&/TEMP_RiftGauntlets_Dash_Impact_Distant_03}}
 {id=WavePlayer_13, properties {soundWaveAssetPtr=&/TEMP_RiftGauntlets_Dash_Impact_Distant_04}}
 {id=WavePlayer_14, properties {soundWaveAssetPtr=&/TEMP_RiftGauntlets_Dash_Impact_Distant_05}}
 {id=WavePlayer_15, properties {soundWaveAssetPtr=&/TEMP_RiftGauntlets_Dash_Impact_Distant_06}}
 {id=WavePlayer_16, properties {soundWaveAssetPtr=&/TEMP_RiftGauntlets_Dash_Impact_Distant_07}}
 {id=WavePlayer_17, properties {soundWaveAssetPtr=&/TEMP_RiftGauntlets_Dash_Impact_Distant_08}}
 {id=WavePlayer_18, properties {soundWaveAssetPtr=&/TEMP_RiftGauntlets_Dash_Impact_Distant_09}}
 {id=WavePlayer_19, properties {soundWaveAssetPtr=&/TEMP_RiftGauntlets_Dash_Impact_Distant_10}}
 {id=WavePlayer_2, properties {soundWaveAssetPtr=&/TEMP_RiftGauntlets_Dash_Impact_Close_03}}
 {id=WavePlayer_20, properties {soundWaveAssetPtr=&/TEMP_RiftGauntlets_Dash_Impact_Zaps_01}}
 {id=WavePlayer_21, properties {soundWaveAssetPtr=&/TEMP_RiftGauntlets_Dash_Impact_Zaps_02}}
 {id=WavePlayer_22, properties {soundWaveAssetPtr=&/TEMP_RiftGauntlets_Dash_Impact_Zaps_03}}
 {id=WavePlayer_23, properties {soundWaveAssetPtr=&/TEMP_RiftGauntlets_Dash_Impact_Zaps_04}}
 {id=WavePlayer_24, properties {soundWaveAssetPtr=&/TEMP_RiftGauntlets_Dash_Impact_Zaps_05}}
 {id=WavePlayer_25, properties {soundWaveAssetPtr=&/TEMP_RiftGauntlets_Dash_Impact_Zaps_06}}
 {id=WavePlayer_26, properties {soundWaveAssetPtr=&TEMP_RiftGauntlets_Dash_Impact_Zaps_01}}
 {id=WavePlayer_27, properties {soundWaveAssetPtr=&TEMP_RiftGauntlets_Dash_Impact_Zaps_02}}
 {id=WavePlayer_28, properties {soundWaveAssetPtr=&TEMP_RiftGauntlets_Dash_Impact_Zaps_03}}
 {id=WavePlayer_29, properties {soundWaveAssetPtr=&TEMP_RiftGauntlets_Dash_Impact_Zaps_04}}
 {id=WavePlayer_3, properties {soundWaveAssetPtr=&/TEMP_RiftGauntlets_Dash_Impact_Close_04}}
 {id=WavePlayer_30, properties {soundWaveAssetPtr=&TEMP_RiftGauntlets_Dash_Impact_Zaps_05}}
 {id=WavePlayer_31, properties {soundWaveAssetPtr=&TEMP_RiftGauntlets_Dash_Impact_Zaps_06}}
 {id=WavePlayer_4, properties {soundWaveAssetPtr=&/TEMP_RiftGauntlets_Dash_Impact_Close_05}}
 {id=WavePlayer_5, properties {soundWaveAssetPtr=&/TEMP_RiftGauntlets_Dash_Impact_Close_06}}
 {id=WavePlayer_6, properties {soundWaveAssetPtr=&/TEMP_RiftGauntlets_Dash_Impact_Close_07}}
 {id=WavePlayer_7, properties {soundWaveAssetPtr=&/TEMP_RiftGauntlets_Dash_Impact_Close_08}}
 {id=WavePlayer_8, properties {soundWaveAssetPtr=&/TEMP_RiftGauntlets_Dash_Impact_Close_09}}
 {id=WavePlayer_9, properties {soundWaveAssetPtr=&/TEMP_RiftGauntlets_Dash_Impact_Close_10}}
]
properties {
 maxDistance=12000
 soundClassObject=/Game/Sounds/FortSoundClassesAndModes/Fort_Explosives
 volumeMultiplier=1
 concurrencyOverrides {maxCount=4, resolutionRule=StopOldest, volumeScale=0.5}
 concurrencySet [/Game/Sounds/Concurrency_Rules/Weapon_Explosion_Conc]
}
